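; Gameplay ability for the Figment Zapper Trap grenade that previews the throw arc and trap placement while aiming.
; On trajectory init it spawns a red arc actor, plus a context trap tool linked to a projectile deco helper.
; Spline updates feed the hit result to the helper, showing the trap preview and normal arc for valid spots, red for invalid actors or positions.
; Cleanup and ability end unequip the preview tool.
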(event TrajectoryInterface|EventInitTrajectoryVariables
  ;[1 1 1] Init Trajectory
  (|CallParentFunction)
  (CallFunction|SetUpDummyProjectileRed)
  (CallFunction|InitializeDecoHelper))

(event Ability|EventOnEndAbility
  (params (bWasCancelled :type Boolean))
  (bind _grenadedecotool (Variables|Default|GetGrenadeDecoTool))
  ;[1 1 1] Event End Ability
  (|CallParentFunction bWasCancelled)
  (CallFunction|EndTrapPreview)
  ;[1 1 1] Function Override, sends trajectory hit result to player pawn after update. Sets red trajectory line if trap surface is invalid
  (Utilities|FlowControl|DoOnce false
    (:Completed
      (Utilities|IsValid _grenadedecotool
        (:Is%20Valid
          (Class|FortTrapTool|OnEquip _grenadedecotool 0)
          (CallFunction|InitializeDecoHelper))))))

(event Trajectory|EventCleanupTrajectory
  (bind _grenadedecotool (Variables|Default|GetGrenadeDecoTool))
  ;[1 1 1] Cleanup Trajectory Display
  (|CallParentFunction)
  (Actor|DestroyActor (Variables|Default|GetProjectileTrajectoryRed))
  (CallFunction|EndTrapPreview)
  ;[1 1 1] Function Override, sends trajectory hit result to player pawn after update. Sets red trajectory line if trap surface is invalid
  (Utilities|FlowControl|DoOnce false
    (:Completed
      (Utilities|IsValid _grenadedecotool
        (:Is%20Valid
          (Class|FortTrapTool|OnEquip _grenadedecotool 0)
          (CallFunction|InitializeDecoHelper))))))

(event Trajectory|EventUpdateTrajectorySpline
  (bind _grenadedecohelper (Variables|Default|GetGrenadeDecoHelper))
  (bind _returnvalue (Class|FortDecoHelper|GetinFallbackPosition _grenadedecohelper))
  (bind _returnvalue_2 (not _returnvalue))
  (bind _grenadedecohelper_2 (Variables|Default|GetGrenadeDecoHelper))
  (bind _returnvalue_3 (Class|FortProjectileDecoHelper|GetInvalidThrowTrapPosition _grenadedecohelper_2))
  (bind _returnvalue_4 (not _returnvalue_3))
  (bind _trajectoryindicator (Variables|Trajectory|GetTrajectoryIndicator))
  (bind _projectiletrajectoryred (Variables|Default|GetProjectileTrajectoryRed))
  (bind _grenadedecotool (Variables|Default|GetGrenadeDecoTool))
  (bind _grenadedecohelper_3 (Variables|Default|GetGrenadeDecoHelper))
  ;[1 1 1] Function Override, sends trajectory hit result to player pawn after update. Sets red trajectory line if trap surface is invalid
  (Utilities|IsValid (Variables|Trajectory|GetTrajectoryIndicator)
    (:Is%20Valid
      (Utilities|IsValid _grenadedecohelper_3
        (:Is%20Valid
          (bind (_outsplinepoints _outsplinetangents _outhitresult) (TrajectoryInterface|GetProjectileTrajectoryPoints))
          (Class|FortProjectileDecoHelper|SetGrenadeTrajectoryHitResult _grenadedecohelper_3 _outhitresult)
          (Trajectory|SetTrajectorySpline (select (Math|Boolean|ANDBoolean (not (and (not (.bSuccess (Utilities|Casting|CastToAthenaFillFloor (.HitActor (Collision|BreakHitResult _outhitresult))))) (Math|Boolean|ORBoolean (.bSuccess (Utilities|Casting|CastToBuildingProp (.HitActor (Collision|BreakHitResult _outhitresult)))) (.bSuccess (Utilities|Casting|CastToFortPawn (.HitActor (Collision|BreakHitResult _outhitresult)))) (.bSuccess (Utilities|Casting|CastToFortAthenaSupplyDrop (.HitActor (Collision|BreakHitResult _outhitresult)))) (.bSuccess (Utilities|Casting|CastToAthenaSupplyDropBalloon (.HitActor (Collision|BreakHitResult _outhitresult)))) (.bSuccess (Utilities|Casting|CastToFortProjectileBase (.HitActor (Collision|BreakHitResult _outhitresult))))))) _returnvalue_2 _returnvalue_4) _trajectoryindicator _projectiletrajectoryred) _outsplinepoints _outsplinetangents)
          (Rendering|SetActorHiddenInGame (select (Math|Boolean|ANDBoolean (not (and (not (.bSuccess (Utilities|Casting|CastToAthenaFillFloor (.HitActor (Collision|BreakHitResult _outhitresult))))) (Math|Boolean|ORBoolean (.bSuccess (Utilities|Casting|CastToBuildingProp (.HitActor (Collision|BreakHitResult _outhitresult)))) (.bSuccess (Utilities|Casting|CastToFortPawn (.HitActor (Collision|BreakHitResult _outhitresult)))) (.bSuccess (Utilities|Casting|CastToFortAthenaSupplyDrop (.HitActor (Collision|BreakHitResult _outhitresult)))) (.bSuccess (Utilities|Casting|CastToAthenaSupplyDropBalloon (.HitActor (Collision|BreakHitResult _outhitresult)))) (.bSuccess (Utilities|Casting|CastToFortProjectileBase (.HitActor (Collision|BreakHitResult _outhitresult))))))) _returnvalue_2 _returnvalue_4) _trajectoryindicator _projectiletrajectoryred) false)
          (Class|FortProjectileDecoHelper|SetShowDecoPreview (Variables|Default|GetGrenadeDecoHelper) (Math|Boolean|ANDBoolean (not (and (not (.bSuccess (Utilities|Casting|CastToAthenaFillFloor (.HitActor (Collision|BreakHitResult _outhitresult))))) (Math|Boolean|ORBoolean (.bSuccess (Utilities|Casting|CastToBuildingProp (.HitActor (Collision|BreakHitResult _outhitresult)))) (.bSuccess (Utilities|Casting|CastToFortPawn (.HitActor (Collision|BreakHitResult _outhitresult)))) (.bSuccess (Utilities|Casting|CastToFortAthenaSupplyDrop (.HitActor (Collision|BreakHitResult _outhitresult)))) (.bSuccess (Utilities|Casting|CastToAthenaSupplyDropBalloon (.HitActor (Collision|BreakHitResult _outhitresult)))) (.bSuccess (Utilities|Casting|CastToFortProjectileBase (.HitActor (Collision|BreakHitResult _outhitresult))))))) _returnvalue_2 _returnvalue_4))
          (Rendering|SetActorHiddenInGame (select (Math|Boolean|ANDBoolean (not (and (not (.bSuccess (Utilities|Casting|CastToAthenaFillFloor (.HitActor (Collision|BreakHitResult _outhitresult))))) (Math|Boolean|ORBoolean (.bSuccess (Utilities|Casting|CastToBuildingProp (.HitActor (Collision|BreakHitResult _outhitresult)))) (.bSuccess (Utilities|Casting|CastToFortPawn (.HitActor (Collision|BreakHitResult _outhitresult)))) (.bSuccess (Utilities|Casting|CastToFortAthenaSupplyDrop (.HitActor (Collision|BreakHitResult _outhitresult)))) (.bSuccess (Utilities|Casting|CastToAthenaSupplyDropBalloon (.HitActor (Collision|BreakHitResult _outhitresult)))) (.bSuccess (Utilities|Casting|CastToFortProjectileBase (.HitActor (Collision|BreakHitResult _outhitresult))))))) _returnvalue_2 _returnvalue_4) (Variables|Default|GetProjectileTrajectoryRed) (Variables|Trajectory|GetTrajectoryIndicator)) true)
          (Utilities|FlowControl|DoOnce false
            (:Completed
              (Utilities|IsValid _grenadedecotool
                (:Is%20Valid
                  (Class|FortTrapTool|OnEquip _grenadedecotool 0)
                  (CallFunction|InitializeDecoHelper))))))))))

(fn EndTrapPreview
  (bind _grenadedecotool (Variables|Default|GetGrenadeDecoTool))
  (Utilities|IsValid _grenadedecotool
    (:Is%20Valid
      (Class|FortDecoTool|OnUnEquip _grenadedecotool))))

(fn InitializeDecoHelper
  (bind _returnvalue (Default|GetOwningPlayer))
  (bind _grenadedecotool (Variables|Default|GetGrenadeDecoTool))
  (Utilities|FlowControl|Sequence
    (:then_0
      (Variables|Default|GetGrenadeDecoTool
        (:then)
        (:else
          (Utilities|IsValid _returnvalue
            (:Is%20Valid
              (bind _returnvalue_2 (Game|SpawnActorfromClass "/Script/FortniteGame.FortDecoTool_ContextTrap" (Transformation|GetActorTransform _returnvalue) "Undefined" "MultiplyWithRoot" _returnvalue (Variables|/Script/Engine.Actor|GetInstigator _returnvalue))) ; Spawn Trap Tool for Trap Preview (Must be context trap)
              (Utilities|IsValid _returnvalue_2
                (:Is%20Valid
                  (Variables|Default|SetGrenadeDecoTool _returnvalue_2))))))))
    (:then_1
      (Variables|Default|GetGrenadeDecoHelper
        (:then)
        (:else
          (bind _returnvalue_3 (Game|SpawnActorfromClass "/Script/FortniteGame.FortProjectileDecoHelper" (Transformation|GetActorTransform (Default|GetOwningPlayer)) "Undefined" "OverrideRootScale" (Default|GetOwningPlayer) (Variables|/Script/Engine.Actor|GetInstigator (Default|GetOwningPlayer))))
          (Utilities|IsValid _returnvalue_3
            (:Is%20Valid
              (Variables|Default|SetGrenadeDecoHelper _returnvalue_3))))))
    (:then_2
      (Data|SetWeaponData (Variables|Default|GetGrenadeDecoTool) (Variables|Default|GetTrapDefinition))
      (Class|FortDecoToolContextTrap|SetContextTrapItemDefinition (Variables|Default|GetGrenadeDecoTool) (Variables|Default|GetTrapDefinition))
      (Variables|/Script/FortniteGame.FortDecoHelper|SetDecoTool (Variables|Default|GetGrenadeDecoTool) (Variables|Default|GetGrenadeDecoHelper))
      (Class|FortDecoTool|SetHelper _grenadedecotool (Variables|Default|GetGrenadeDecoHelper))
      (Class|FortDecoTool|SetDecoObjectPreview _grenadedecotool (Variables|Default|GetTrapDefinition) true))))

(fn SetUpDummyProjectileRed
  (bind _returnvalue (Game|SpawnActorfromClass (Variables|Default|GetTrajectoryIndicatorRedClass) (Transformation|GetActorTransform (Variables|ReferencedActors|GetPlayerPawn)) "AlwaysSpawn" "SelectDefaultAtRuntime" 0 (Variables|ReferencedActors|GetPlayerPawn)))
  (bind _asbp_projectile_trajectory (Utilities|Casting|CastToBP_ProjectileTrajectory _returnvalue))
  (Variables|Default|SetProjectileTrajectoryRed _asbp_projectile_trajectory))
